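# Nanite-ready console screen material that shows an animated UI image with a CRT pattern, scanlines and a checkmark overlay.
# UVs come from local-position XZ, feeding the screen texture, hue shift and distance-field checkmark.
# BP_GetTime-relative timers drive invertible on/off and checkmark animations, alongside a height-based wipe mask and square vignette.
# Emissive divides out eye adaptation and adds a power-curve highlight over a dark, glossy base.
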Material /FlexLegendsENVGameplay/Assets/Materials/M_FlexLegend_ConsoleScreen
graph {
 properties {bAutomaticallySetUsageInEditor=false, bUseMaterialAttributes=true, bUsedWithNanite=true, usedShadingModels=Default Lit}
 nodes [
  n1 = NamedRerouteUsage()
  n2 = LocalPosition()
  n3 = ComponentMask(None=@n2.XYZ) {b=true, r=true}
  n4 = VectorParameter() {group=Basic_PBR, parameterName=UV_Scale/Offset}
  n5 = ComponentMask(None=@n4.RGB) {g=true, r=true}
  n6 = Constant()
  n7 = Add(
   A=Multiply(A=Multiply(A=@n3) {constB=-1}, B=@n5) {constB=0.01}
   B=Multiply(A=@n5, B=MaterialFunctionCall(R=@n4.B, G=@n4.B, B=@n6) {materialFunction=&AppendMany}.RG)
  )
  n8 = TextureSample(UVs=MaterialFunctionCall(UVs=@n7, Texture Scale=Constant() {r=0.015}) {materialFunction=&ScaleUVsByCenter}.UVs) {samplerType=SAMPLERTYPE_LinearColor, texture=/Game/Accessories/FORT_Backpacks/Backpack_M_MED_Jumbotron_S20/Textures/T_Jumbotron_CRT}
  n9 = Add(
   A=MaterialFunctionCall!(ScanLine_Tiling_1, Scanline_PannerSpeed_1, ScanLine_Tiling, Scanline_PannerSpeed, Scanline_Rotation, UVs=@n7) {materialFunction=/Game/Global/Materials/MaterialGym/Materials/HolographicTexture/MF_HolographicTexture_ScanLines}.Result
   B=ScalarParameter() {group=Basic_PBR, parameterName=ScanLines_Min}
  )
  n10 = Saturate(None=Multiply(A=NamedRerouteUsage()) {constB=1.35})
  n11 = MaterialFunctionCall(UVs=@n7, Texture Scale=AppendVector(A=LinearInterpolate(Alpha=@n10) {constA=0.8}, B=LinearInterpolate(Alpha=@n10) {constA=0.23})) {materialFunction=&ScaleUVsByCenter}
  n12 = TextureCoordinate()
  n13 = TextureSampleParameter2D!(UVs=@n11.UVs, "DDX(UVs)"=DDX(Value=@n12), "DDY(UVs)"=DDY(Value=@n12), Apply View MipBias) {
   group=Basic_PBR
   mipValueMode=TMVM_Derivative
   parameterName=ScreenColor
   samplerSource=SSM_Clamp_WorldGroupSettings
   texture=/BarnTowelArmoryGameplay/Assets/InteractiveScreen/T_UI_InteractiveScreemIM
  }
  n14 = Add(A=@n13.RGB) {constB=0.01}
  n15 = ScalarParameter() {defaultValue=0.5, group=ScreenAnimation, parameterName=Checkmark_HueShiftPercentage}
  n16 = VectorParameter() {group=ScreenAnimation, parameterName=ChechmarkColor_Circle, sortPriority=30, defaultValue=<rgba 0 0.3864 0.3763 1>}
  n17 = Constant() {r=0.433}
  n18 = Multiply(A=Constant2Vector() {g=1, r=-1}, B=LinearInterpolate(B=@n17, Alpha=NamedRerouteUsage()) {constA=0.8, constAlpha=1, constB=0.8}) {constB=0.45}
  n19 = MaterialFunctionCall(UVs=@n7, Texture Scale=@n18) {materialFunction=&ScaleUVsByCenter}
  n20 = SmoothStep(Value=TextureSample(UVs=@n19.UVs) {samplerSource=SSM_Clamp_WorldGroupSettings, texture=/FlexLegendsENVGameplay/Assets/Textures/T_StarkScreen_Checkmark_DF}.R) {constMax=0.105, constMin=0.1}
  n21 = Multiply(A=OneMinus(None=@n20), B=SphereMask(A=@n19.UVs, B=Constant() {r=0.5}, Radius=@n17) {attenuationRadius=0.4})
  n22 = Multiply(A=Saturate(None=Add(A=@n20, B=@n21)), B=NamedRerouteUsage())
  n23 = LinearInterpolate(
   A=LinearInterpolate(
    A=@n14
    B=MaterialFunctionCall(Hue Shift Percentage=@n15, Texture=@n14) {materialFunction=&HueShift}.Result
    Alpha=NamedRerouteUsage()
   )
   B=LinearInterpolate(A=VectorParameter() {parameterName=CheckmarkColor_Symbol, defaultValue=<rgba 0.95 0.95 0.95 1>}.RGB, B=@n16.RGB, Alpha=@n20) {constA=0.95}
   Alpha=@n22
  )
  n24 = Multiply(
   A=Multiply(A=Saturate(None=@n9), B=Add(A=@n23, B=@n23))
   B=LinearInterpolate(A=ScalarParameter() {defaultValue=1, group=Basic_PBR, parameterName=ScreenBrightness}, Alpha=@n22) {constB=3}
  )
  n25 = ScalarParameter() {group=ScreenAnimation, parameterName=InvertScreenAnimation}
  n26 = LinearInterpolate(B=Add(B=Multiply(A=@n25) {constB=3}) {constA=1}, Alpha=NamedRerouteUsage()) {constA=10}
  n27 = OneMinus(None=Floor(None=NamedRerouteUsage()))
  n28 = Constant() {r=0.05}
  n29 = Add(
   A=Multiply(A=Multiply(A=@n8.RGB, B=@n24), B=LinearInterpolate(B=@n26, Alpha=@n27) {constA=1})
   B=Multiply(A=Multiply(A=@n25, B=@n28), B=@n27)
  )
  n30 = Multiply(A=Divide(A=Multiply(A=@n1, B=@n29), B=EyeAdaptation().EyeAdaptation), B=OneMinus(None=NamedRerouteUsage()))
  n31 = Multiply(A=Power(Base=Multiply(B=@n30) {constA=0.1}) {constExponent=16}) {constB=0.1}
  MakeMaterialAttributes!(
   BaseColor=Add(A=NamedRerouteUsage(), B=VectorParameter() {group=Basic_PBR, parameterName=BaseColor, defaultValue=<rgba 0 0 0 1>}.RGB)
   Metallic=ScalarParameter() {group=Basic_PBR, parameterName=Metallic}
   Specular=ScalarParameter() {defaultValue=0.5, group=Basic_PBR, parameterName=Specular}
   Roughness=ScalarParameter() {defaultValue=0.1, group=Basic_PBR, parameterName=Roughness}
   Anisotropy
   EmissiveColor=Add(A=NamedRerouteDeclaration(None=@n31) {name=Highlights, nodeColor=<rgba 1 0 0.1406 1>}, B=@n30)
   Opacity
   OpacityMask
   Normal
   Tangent
   WorldPositionOffset
   SubsurfaceColor
   ClearCoat
   ClearCoatRoughness
   AmbientOcclusion
   Refraction
   CustomizedUVs_0
   CustomizedUVs_1
   CustomizedUVs_2
   CustomizedUVs_3
   CustomizedUVs_4
   CustomizedUVs_5
   CustomizedUVs_6
   CustomizedUVs_7
   PixelDepthOffset
   ShadingModel
   Displacement
  )
  Multiply(B=@n18)
  n32 = ScalarParameter() {defaultValue=1, group=ScreenAnimation, parameterName=Screen_Hardness}
  n33 = Add(A=ComponentMask(None=@n2.XYZ) {b=true}, B=ScalarParameter() {group=ScreenAnimation, parameterName=Screen_HeightOffset}) {constB=0}
  n34 = LinearInterpolate(
   A=ScalarParameter() {defaultValue=-90, group=ScreenAnimation, parameterName=Screen_Min}
   B=ScalarParameter() {defaultValue=100, group=ScreenAnimation, parameterName=Screen_Max}
   Alpha=NamedRerouteUsage()
  ) {constA=-90, constB=100}
  n35 = Add(A=Add(A=@n33, B=@n34), B=ScalarParameter() {defaultValue=100, group=ScreenAnimation, parameterName=Screen_HeightCorrection}) {constB=100}
  n36 = OneMinus(None=Saturate(None=Multiply(A=@n32, B=@n35) {constA=0.01}))
  n37 = Multiply(A=@n36, B=@n36)
  n38 = Multiply(A=@n37, B=@n37)
  n39 = Multiply(B=@n34) {constA=-1}
  n40 = Saturate(None=Multiply(A=Add(A=@n33, B=@n39), B=@n32) {constB=0.01})
  n41 = Multiply(A=@n40, B=@n40)
  n42 = Multiply(A=@n41, B=@n41)
  NamedRerouteDeclaration(None=Saturate(None=Add(A=@n38, B=@n42) {constB=0})) {name=OnOff_Mask, nodeColor=<rgba 0.8984 0 1 1>}
  n43 = Subtract(
   A=Time()
   B=Multiply(
    A=ScalarParameter() {group=ScreenAnimation, parameterName=InvertScreenAnimation}
    B=ScalarParameter() {group=ScreenAnimation, parameterName=TurnOffAnimationDelay}
   )
  )
  n44 = ScalarParameter() {defaultValue=10000000000, group=ScreenAnimation, parameterName=BP_GetTime}
  n45 = ScalarParameter() {defaultValue=1, group=ScreenAnimation, parameterName=ScreenSpeed}
  n46 = Multiply(A=@n45) {constB=2.5}
  n47 = ScalarParameter() {group=ScreenAnimation, parameterName=InvertScreenAnimation}
  n48 = Saturate(None=Multiply(A=Subtract(A=@n43, B=@n44), B=LinearInterpolate(A=@n45, B=@n46, Alpha=@n47)))
  n49 = OneMinus(None=@n48)
  n50 = Multiply(A=Multiply(A=@n49, B=@n49), B=@n49)
  NamedRerouteDeclaration(None=LinearInterpolate(A=OneMinus(None=@n50), B=OneMinus(None=@n48), Alpha=ScalarParameter() {group=ScreenAnimation, parameterName=InvertScreenAnimation}) {constAlpha=0}) {name=AnimatedTime, nodeColor=<rgba 1 0 0.4687 1>}
  n51 = Subtract(
   A=Subtract(A=Time(), B=ScalarParameter() {group=ScreenAnimation, parameterName=CheckmarkDelay})
   B=ScalarParameter() {defaultValue=10000000000, group=ScreenAnimation, parameterName=BP_GetTime}
  )
  n52 = Saturate(None=Multiply(A=@n51, B=ScalarParameter() {defaultValue=1, group=ScreenAnimation, parameterName=CheckmarkSpeed_Scale}))
  n53 = Abs(None=Multiply(A=Add(A=@n52) {constB=-0.5}) {constB=2})
  n54 = Multiply(A=@n53, B=@n53)
  NamedRerouteDeclaration(None=Multiply(A=@n54, B=ScalarParameter() {group=ScreenAnimation, parameterName=InvertScreenAnimation})) {name=AnimatedTime_CheckmarkScale, nodeColor=<rgba 0 0.1562 1 1>}
  n55 = Multiply(A=@n51, B=ScalarParameter() {defaultValue=1, group=ScreenAnimation, parameterName=CheckmarkSpeed_Opacity})
  NamedRerouteDeclaration(None=Multiply(A=Saturate(None=@n55), B=ScalarParameter() {group=ScreenAnimation, parameterName=InvertScreenAnimation})) {name=AnimatedTime_CheckmarkOpacity, nodeColor=<rgba 1 0.5625 0 1>}
  n56 = Time()
  n57 = Saturate(None=Multiply(A=Subtract(A=@n56), B=ScalarParameter() {defaultValue=1, group=ScreenAnimation, parameterName=LogoSpeed_Rotation}))
  n58 = Multiply(A=@n57, B=@n57)
  NamedRerouteDeclaration(None=Multiply(A=Multiply(A=@n58, B=@n57), B=ScalarParameter() {group=ScreenAnimation, parameterName=InvertScreenAnimation})) {name=AnimatedTime_LogoRotation, nodeColor=<rgba 1 0.6328 0 1>}
  n59 = VectorParameter() {group=Basic_PBR, parameterName=VignetteMaskCoord}
  n60 = MaterialFunctionCall!(
   Texture Coordinate=@n11.UVs
   X Pos Right=@n59.R
   X Pos Left=@n59.R
   Y Pos Up=@n59.R
   Y Pos Down=@n59.R
   SquareSoftness=ScalarParameter() {group=Basic_PBR, parameterName=VignetteSoftness}
   RotateSquare
   RotationAmount
  ) {materialFunction=/Game/Accessories/FORT_Backpacks/Backpack_F_MED_GroovyReader_Play/Materials/MaterialFunctions/MF_SquareMaskGenerator}
  NamedRerouteDeclaration(None=@n60.SquareMask) {name=Vignette, nodeColor=<rgba 1 0 0.5391 1>}
  OneMinus(None=Multiply(A=@n54, B=OneMinus(None=@n53)))
  Reroute(None=@n60.SquareMask)
  SphereMask()
  TextureCoordinate()
 ]
}
