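# Parent material for the Flex Legends platform, built on the shared character material functions with a timed emissive on/off fade.
# Character PreFX feeds CharacterSkinFX, whose emissive is scaled before passing through the base and cinematic PostFX functions.
# A smoothstep over time since Blueprint-driven BP_GetTime, scaled by Emissive_Speed, blends between minimum and maximum brightness.
# The TurnOn/Off index flips the fade direction, and the huge default timestamp holds it at its start state.
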Material /FlexLegendsENVGameplay/Assets/Materials/M_FlexLegend_platform_PARENT
graph {
 properties {
  bAutomaticallySetUsageInEditor=false
  bUseMaterialAttributes=true
  bUsedWithClothing=true
  bUsedWithMeshDeformer=true
  bUsedWithMorphTargets=true
  bUsedWithNanite=true
  bUsedWithSkeletalMesh=true
  pixelDepthOffsetMode=PDOM_Legacy
  refractionMethod=RM_IndexOfRefraction
  shadingModel=MSM_FromMaterialExpression
  subsurfaceProfile=&SS_HeroSkin_02
  usedShadingModels=Default Lit | Subsurface | Clear Coat | Subsurface Profile | Cloth | Eye
 }
 nodes [
  n1 = MaterialFunctionCall() {materialFunction=&BaseFN_CharacterFunction_PreFX}
  n2 = GetMaterialAttributes(None=MaterialFunctionCall(
    PreskinnedPosition=@n1.Pre skinned Local Position
    Gmap=@n1.Pre skinned Local Position
    MaterialAttributes=@n1.Pre skinned Local Position
    AO=@n1.Pre skinned Local Position
   ) {materialFunction=&CharacterSkinFX}.Result) {attributeGetTypes [#1]}
  n3 = Time()
  n4 = ScalarParameter() {defaultValue=10000000000, group=BlueprintLogic, parameterName=BP_GetTime, sortPriority=0}
  n5 = SmoothStep(Value=Multiply(A=Subtract(A=@n3, B=@n4), B=ScalarParameter() {defaultValue=1, group=BlueprintLogic, parameterName=Emissive_Speed}))
  n6 = LinearInterpolate(
   A=ScalarParameter() {group=BlueprintLogic, parameterName=EmissiveBrightness_Min}
   B=ScalarParameter() {defaultValue=1, group=BlueprintLogic, parameterName=EmissiveMaxBrightness}
   Alpha=LinearInterpolate(
    A=OneMinus(None=@n5)
    B=@n5
    Alpha=ScalarParameter() {defaultValue=1, group=BlueprintLogic, parameterName=Emissive_TurnOn/Off_Index, sliderMax=1}
   )
  )
  n7 = MaterialFunctionCall(
   AO=@n1.AO
   Distance Blend=@n1.AO
   Fresnel=@n1.AO
   Gmap=@n1.AO
   Pre Skinned Local Position=@n1.AO
   Material=SetMaterialAttributes(MaterialAttributes=@n2.MaterialAttributes, Emissive Color=Multiply(A=@n2.Emissive Color, B=@n6)) {attributeSetTypes [#1]}
  ) {materialFunction=&BaseFN_CharacterFunction_PostFX}
  MaterialFunctionCall(In=@n7.Result) {materialFunction=&CINE_MF_CharPostFX}
 ]
}
