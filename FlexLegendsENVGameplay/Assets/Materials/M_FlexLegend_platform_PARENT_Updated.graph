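# Parent material for Flex Legends platforms, built on the 2023 character parent PreFX and PostFX functions, that adds a timed emissive on/off fade.
# It measures time elapsed since a Blueprint-supplied timestamp, scales it by a speed parameter and smoothsteps it.
# An index parameter blends between the ramp and its inverse to pick turn-on or turn-off.
# The result blends minimum to maximum brightness, multiplying emissive color before PostFX runs.
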Material /FlexLegendsENVGameplay/Assets/Materials/M_FlexLegend_platform_PARENT_Updated
graph {
 properties {
  bUseMaterialAttributes=true
  bUsedWithClothing=true
  bUsedWithMeshDeformer=true
  bUsedWithNanite=true
  bUsedWithNiagaraMeshParticles=true
  bUsedWithSkeletalMesh=true
  pixelDepthOffsetMode=PDOM_Legacy
  shadingModel=MSM_FromMaterialExpression
  usedShadingModels=Default Lit | Subsurface | Clear Coat | Subsurface Profile | Cloth
 }
 nodes [
  n1 = MaterialFunctionCall() {materialFunction=&MF_CharacterParent_2023_PreFX}
  n2 = GetMaterialAttributes(None=@n1.Result) {attributeGetTypes [#1]}
  n3 = Time()
  n4 = ScalarParameter() {defaultValue=10000000000, group=BlueprintLogic, parameterName=BP_GetTime, sortPriority=0}
  n5 = SmoothStep(Value=Multiply(A=Subtract(A=@n3, B=@n4), B=ScalarParameter() {defaultValue=1, group=BlueprintLogic, parameterName=Emissive_Speed}))
  n6 = LinearInterpolate(
   A=ScalarParameter() {group=BlueprintLogic, parameterName=EmissiveBrightness_Min}
   B=ScalarParameter() {defaultValue=1, group=BlueprintLogic, parameterName=EmissiveMaxBrightness}
   Alpha=LinearInterpolate(
    A=OneMinus(None=@n5)
    B=@n5
    Alpha=ScalarParameter() {defaultValue=1, group=BlueprintLogic, parameterName=Emissive_TurnOn/Off_Index, sliderMax=1}
   )
  )
  MaterialFunctionCall(
   Material=SetMaterialAttributes(MaterialAttributes=@n2.MaterialAttributes, Emissive Color=Multiply(A=@n2.Emissive Color, B=@n6)) {attributeSetTypes [#1]}
   M=@n1.M
   Pre Skinned Local Position=@n1.M
   Distance Blend=@n1.M
   Fresnel=@n1.M
  ) {materialFunction=&MF_CharacterParent_2023_PostFX}
 ]
}
