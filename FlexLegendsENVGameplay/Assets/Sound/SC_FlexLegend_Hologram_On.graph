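# Sound cue for a hologram switching on at a FlexLegend stand, in the FlexLegends environment gameplay assets.
# A single wave player feeds the "Stand Hologram On" wave straight to the output, with no randomization or mixing.
# It plays boosted in volume and slightly pitched down in the gadgets sound class, spatialized through the attenuation settings named for FlexLegend mech arms.
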SoundCue /FlexLegendsENVGameplay/Assets/Sound/SC_FlexLegend_Hologram_On
root=WavePlayer_0
nodes [{id=WavePlayer_0, properties {soundWaveAssetPtr=/FlexLegendsENVGameplay/Assets/Sound/wav/SC_FlexLegend_Stand_Hologram_On}}]
properties {
 attenuationSettings=&/AS_FlexLegend_MechArms
 maxDistance=1700
 pitchMultiplier=0.9
 soundClassObject=&Fort_Gadgets
 volumeMultiplier=2.4
}
