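# Sound cue for the music stinger that plays when a FlexLegend stand interaction completes, judging by its name.
# A mixer layers a treasure chest sound, slightly lowered and faded down by a volume envelope, with a BarnTowelArmory door stinger that enters after a fixed delay.
# It uses the music stinger sound class, overrides attenuation inline, and limits concurrency to one instance, stopping the oldest.
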SoundCue /FlexLegendsENVGameplay/Assets/Sound/SC_FlexLegend_MX_Stand_Interact_Complete_Stinger
root=Mixer_0
nodes [
 {id=Delay_0, children [WavePlayer_0], properties {delayMax=1.5, delayMin=1.5}}
 {
  id=Enveloper_0
  children [WavePlayer_2]
  properties {pitchCurve {editorCurveData {keys [{time=0}]}}, volumeCurve {editorCurveData {keys [<at 0 1>,<at 1 1>,<at 3 0.3>]}}}
 }
 {id=Mixer_0, children [Delay_0,Enveloper_0], properties {inputVolume [1,0.7]}}
 {id=WavePlayer_0, properties {soundWaveAssetPtr=/BarnTowelArmoryGameplay/Assets/Sounds/wav/BarnTowelArmory_MX_Stinger_DoorOpen_Close}}
 {id=WavePlayer_2, properties {soundWaveAssetPtr=/Game/Sounds/Foley_Loot/Containers/Treasure_Chest/treasure_chest_sweet_shorter_02}}
]
properties {
 attenuationSettings=&/AS_FlexLegend_MechArms
 bOverrideAttenuation=true
 bOverrideConcurrency=true
 maxDistance=4200
 soundClassObject=/Game/Sounds/FortSoundClassesAndModes/Fort_Music_Stinger_PSM
 volumeMultiplier=1
 attenuationOverrides {stereoSpread=1000, attenuationShapeExtents {x=600}}
 concurrencyOverrides {maxCount=1, resolutionRule=StopOldest, retriggerTime=5, volumeScale=0.3}
}
