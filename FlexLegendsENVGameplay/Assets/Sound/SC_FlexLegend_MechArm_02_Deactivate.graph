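# Sound cue for the deactivation sound of the FlexLegend MechArm 02, part of the FlexLegendsENVGameplay content.
# A single wave player feeds the output directly, with no randomization, mixing or modulation nodes.
# It uses the shared mech arms attenuation asset for spatial falloff and is assigned to the Fort_Gadgets sound class.
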SoundCue /FlexLegendsENVGameplay/Assets/Sound/SC_FlexLegend_MechArm_02_Deactivate
root=WavePlayer_0
nodes [{id=WavePlayer_0, properties {soundWaveAssetPtr=/FlexLegendsENVGameplay/Assets/Sound/wav/SC_FlexLegend_MechArm_02_Deactivate}}]
properties {
 attenuationSettings=&/AS_FlexLegend_MechArms
 maxDistance=1700
 soundClassObject=&Fort_Gadgets
}
