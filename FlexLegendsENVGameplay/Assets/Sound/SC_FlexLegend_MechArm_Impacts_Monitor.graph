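# Sound cue for monitor impacts from the FlexLegend mech arm, built from Fortnite's computer demolition impact sounds.
# A Random node picks one of three equally weighted wave players at reduced volume, routed to the impacts sound class and SFX submix.
# It uses the shared small-object demolition hit attenuation.
# A concurrency override caps voices per owner, blocks new plays at the limit, throttles retriggers and scales down overlapping volume.
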SoundCue /FlexLegendsENVGameplay/Assets/Sound/SC_FlexLegend_MechArm_Impacts_Monitor
root=Random_5
nodes [
 {id=Random_5, children [WavePlayer_0, WavePlayer_1, WavePlayer_15], properties {weights [1,1,1]}}
 {id=WavePlayer_0, properties {soundWaveAssetPtr=/Game/Sounds/Fort_Impact_Sounds/Demolition/Computer/Computer_Impact_01}}
 {id=WavePlayer_1, properties {soundWaveAssetPtr=/Game/Sounds/Fort_Impact_Sounds/Demolition/Computer/Computer_Impact_02}}
 {id=WavePlayer_15, properties {soundWaveAssetPtr=/Game/Sounds/Fort_Impact_Sounds/Demolition/Computer/Computer_Impact_03}}
]
properties {
 attenuationSettings=/Game/Sounds/Attenuation/Fort_Demolition_Hit_Small_Sized_Object_Att
 bOverrideConcurrency=true
 maxDistance=4000
 soundClassObject=&Fort_Impacts
 soundSubmixObject=&Fortnite_SFX
 volumeMultiplier=0.544
 attenuationOverrides {falloffDistance=4650, attenuationShapeExtents {x=800}}
 concurrencyOverrides {
  bLimitToOwner=true
  bVolumeScaleCanRelease=true
  maxCount=2
  resolutionRule=PreventNew
  retriggerTime=1.2
  voiceStealReleaseTime=0.4
  volumeScale=0.3
 }
}
