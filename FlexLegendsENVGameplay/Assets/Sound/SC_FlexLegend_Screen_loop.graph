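# Sound cue that plays a looping computer screen sound for the FlexLegends environment gameplay content.
# A single looping wave player is the root and plays the ComputerScreen loop wave with no other processing nodes.
# Playback runs at reduced volume in the Fort_Gadgets sound class and borrows the MechArms attenuation asset for its spatial falloff.
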SoundCue /FlexLegendsENVGameplay/Assets/Sound/SC_FlexLegend_Screen_loop
root=WavePlayer_0
nodes [{id=WavePlayer_0, properties {bLooping=true, soundWaveAssetPtr=/FlexLegendsENVGameplay/Assets/Sound/wav/SC_FlexLegend_ComputerScreen_Loop}}]
properties {
 attenuationSettings=&/AS_FlexLegend_MechArms
 maxDistance=1700
 soundClassObject=&Fort_Gadgets
 volumeMultiplier=0.5
}
