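# Sound cue for a treasure chest opening on a FlexLegend stand, using shared loot-container foley.
# A Random node picks between two equally weighted "sweet shorter" chest wave variations.
# It plays through item-interact attenuation and the loot reward sound class, with overridden concurrency that caps simultaneous instances, stops the oldest, and applies volume scaling.
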SoundCue /FlexLegendsENVGameplay/Assets/Sound/SC_FlexLegend_Stand_ChestOpen_WM
root=Random_0
nodes [
 {id=Random_0, children [WavePlayer_2,WavePlayer_3], properties {weights [1,1]}}
 {id=WavePlayer_2, properties {soundWaveAssetPtr=/Game/Sounds/Foley_Loot/Containers/Treasure_Chest/treasure_chest_sweet_shorter_02}}
 {id=WavePlayer_3, properties {soundWaveAssetPtr=/Game/Sounds/Foley_Loot/Containers/Treasure_Chest/treasure_chest_sweet_shorter_03}}
]
properties {
 attenuationSettings=/Game/Sounds/Attenuation/Fort_Item_Interact_Att
 bOverrideConcurrency=true
 maxDistance=1850
 soundClassObject=&Fort_Loot_Reward_PSM
 volumeMultiplier=0.9
 concurrencyOverrides {maxCount=3, resolutionRule=StopOldest, retriggerTime=0.25, volumeScale=0.3}
}
