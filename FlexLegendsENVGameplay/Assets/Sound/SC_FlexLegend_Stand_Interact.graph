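# Sound cue for interacting with the Flex Legend stand, playing a robo-arm interact sound.
# A single wave player feeds the output directly, with reduced volume, item-interact attenuation and the gadgets sound class.
# Its own concurrency override caps simultaneous instances, stops the oldest first, sets a retrigger time and applies volume scaling to overlapping plays.
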SoundCue /FlexLegendsENVGameplay/Assets/Sound/SC_FlexLegend_Stand_Interact
root=WavePlayer_0
nodes [{id=WavePlayer_0, properties {soundWaveAssetPtr=/FlexLegendsENVGameplay/Assets/Sound/wav/FlexLegend_Stand_RoboArm_Interact}}]
properties {
 attenuationSettings=/Game/Sounds/Attenuation/Fort_Item_Interact_Att
 bOverrideConcurrency=true
 maxDistance=1850
 soundClassObject=&Fort_Gadgets
 volumeMultiplier=0.5
 concurrencyOverrides {maxCount=3, resolutionRule=StopOldest, retriggerTime=0.25, volumeScale=0.3}
}
